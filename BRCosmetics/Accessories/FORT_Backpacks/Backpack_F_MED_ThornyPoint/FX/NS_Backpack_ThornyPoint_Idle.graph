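# Niagara system for the ThornyPoint backpack that sheds drifting petal sprites from two offset points.
# System.bIsRunning swaps spawning between sparse NE_Petals_Idle and denser NE_Petals_Run, which adds launch velocity, stronger curl noise and inherited source movement.
# Both use gravity, drag and mesh-oriented rotation, with time-of-day tinting.
# User.IsSkydiving zeroes alpha, and User.bIsFrontEnd alters the idle emitter's daytime tint.
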NiagaraSystem /BRCosmetics/Accessories/FORT_Backpacks/Backpack_F_MED_ThornyPoint/FX/NS_Backpack_ThornyPoint_Idle
dependencies {
 usedDataInterfaces [&NiagaraDataInterfaceCurve]
 usedDynamicInputs [
  &SetBoolByFloatComparison
  &VectorLength
  &Multiply_Vector
 ]
 usedModules [
  &SystemState
  &self:SystemSpawnScript.SystemScriptSource.SystemScriptGraph.NiagaraNodeAssignment_1.SetVariables_#1
 ]
}
emitters {
 NE_Petals_Idle {
  emitterUpdateScript [
   &EmitterState(
    Inactive Response=::NewEnumerator1
    Life Cycle Mode=::NewEnumerator1
    Loop Behavior=::NewEnumerator1
    Loop Duration Mode=::NewEnumerator1
    UseLoopDelay=false
   )
   &SpawnRate(
    Spawn Probability=0.2
    SpawnGroup=1
    SpawnRate=&Lerp_Float(A=3, B=0, Alpha=&MakeCustomFloatFromBool(Bool=$System.bIsRunning, True Float=1, False Float=0))
    Use Spawn Probability=true
   ) {module=SpawnRate001}
  ]
  particleSpawnScript [
   &InitializeParticle(
    Lifetime Max=&Lerp_Float(A=3, B=1, Alpha=&MakeCustomFloatFromBool(Bool=$System.bIsRunning, True Float=1, False Float=0))
    Lifetime Min:NiagaraFloat=&Lerp_Float(A=2, B=0.5, Alpha=&MakeCustomFloatFromBool(Bool=$System.bIsRunning, True Float=1, False Float=0))
    Lifetime Mode=::NewEnumerator1
    Position Offset=&MergeHelper_VectorSwitch(0=<10 -8 2>, 1=<-7 -10 5>, 2=?, 3=?, Index=&RandomRangeInt(Minimum=0, Maximum=1), Num=::NewEnumerator0)
    Position Offset Coordinate Space=::Local
    Sprite Rotation Mode=::NewEnumerator2
    Sprite Size Max=<7 7>
    Sprite Size Min=<3 3>
    Sprite Size Mode=::NewEnumerator2
    UsePositionOffset=true
    Write Material Random=true
   )
   &ShapeLocation(MaskSpawnGroup=false, Override Seed=false, Sphere Radius=3, UseOwnerScale=true)
  ]
  particleUpdateScript [
   &ParticleState()
   &GravityForce(
    Coordinate Space=::Local
    Gravity=&RandomRangeVector(
     Minimum=&MakeVector(
      X=0
      Y=&FloatFromCurve(
       FloatCurve=DataInterface() {propertyValues {Curve {
          defaultValue=1
          keys [{interpMode=RCIM_Cubic, tangentMode=RCTM_User, time=0, value=0}, {interpMode=RCIM_Cubic, tangentMode=RCTM_User, time=1, value=1}]
         }}}
       Scale Curve=-500
      )
      Z=-250
     )
     Maximum=&MakeVector(
      X=0
      Y=&FloatFromCurve(
       FloatCurve=DataInterface() {propertyValues {Curve {
          defaultValue=1
          keys [{interpMode=RCIM_Cubic, tangentMode=RCTM_User, time=0, value=0}, {interpMode=RCIM_Cubic, tangentMode=RCTM_User, time=1, value=1}]
         }}}
       Scale Curve=-750
      )
      Z=-500
     )
    )
   )
   &CurlNoiseForce(Bias Noise=false, Noise Frequency=5, Noise Strength=150, Pan Noise=false)
   &AerodynamicDrag(
    Rotation Strength:NiagaraFloat=&RandomRangeFloat(Minimum=5, Maximum=8)
    Use Linear Drag=true
    Use Rotational Drag=true
   )
   &SolveForcesAndVelocity(Clamp Velocity=false, Limit Acceleration=false)
   &ScaleSpriteSize(Uniform Curve Sprite Scale=DataInterface() {propertyValues {Curve {defaultValue=0, keys [{tangentMode=RCTM_User, time=0, value=0}, <at 0.1 1>, <at 0.5 1>, {tangentMode=RCTM_User, time=1, value=0}]}}})
   &ScaleColor(
    Scale Alpha=&FloatFromCurve(
     FloatCurve=DataInterface() {propertyValues {Curve {keys [<at 0 0>,<at 0.1 1>]}}}
     Scale Curve=&Lerp_Float(A=1, B=0, Alpha=&MakeCustomFloatFromBool(Bool=$User.IsSkydiving, True Float=1, False Float=0))
    )
    Scale RGB=&MakeVectorFromLinearColorRGB(Color=&MakeCustomLinearColorFromBool(
      Bool=$User.bIsFrontEnd
      True Color=&FNTimeOfDayColor(
       Night Color=<rgba 0.01 0.0167 0.05 1>
       Morning Color=<rgba 0.05 0.0406 0.0375 1>
       Day Color=<rgba 0.02 0.025 0.05 1>
       Evening Color=<rgba 0.03 0.018 0.018 1>
      )
      False Color=&FNTimeOfDayColor(
       Night Color=<rgba 0.01 0.0167 0.05 1>
       Morning Color=<rgba 0.05 0.0406 0.0375 1>
       Day Color=<rgba 0.0175 0.0187 0.025 1>
       Evening Color=<rgba 0.03 0.018 0.018 1>
      )
     ))
    ScaleA=true
    ScaleRGB=true
   )
   &AlignSpriteToMeshOrientation(SetAlignment=true, SetFacing=true)
   &SpriteRotationRate(Rotation Rate=&RandomRangeFloat(Minimum=-15, Maximum=15))
   &CameraOffset(Camera Offset Amount=&FloatFromCurve(
     FloatCurve=DataInterface() {propertyValues {Curve {
        defaultValue=1
        keys [{interpMode=RCIM_Cubic, tangentMode=RCTM_User, time=0, value=0}, {interpMode=RCIM_Cubic, tangentMode=RCTM_User, time=1, value=1}]
       }}}
     Scale Curve=10
    ))
  ]
  rendererClasses [&NiagaraSpriteRendererProperties]
  renderers [{rendererClass=&NiagaraSpriteRendererProperties}]
 }
 NE_Petals_Run {
  emitterUpdateScript [
   &EmitterState(
    Inactive Response=::NewEnumerator1
    Life Cycle Mode=::NewEnumerator1
    Loop Behavior=::NewEnumerator1
    Loop Duration Mode=::NewEnumerator1
    UseLoopDelay=false
   )
   &SpawnRate(
    Spawn Probability=0.2
    SpawnGroup=1
    SpawnRate=&Lerp_Float(A=0, B=12, Alpha=&MakeCustomFloatFromBool(Bool=$System.bIsRunning, True Float=1, False Float=0))
    Use Spawn Probability=true
   ) {module=SpawnRate001}
  ]
  particleSpawnScript [
   &InitializeParticle(
    Lifetime Max=1
    Lifetime Min=0.5
    Lifetime Mode=::NewEnumerator1
    Position Offset=&MergeHelper_VectorSwitch(0=<10 -8 2>, 1=<-7 -10 5>, 2=?, 3=?, Index=&RandomRangeInt(Minimum=0, Maximum=1), Num=::NewEnumerator0)
    Position Offset Coordinate Space=::Local
    Sprite Rotation Mode=::NewEnumerator2
    Sprite Size Max=<6 6>
    Sprite Size Min=<3 3>
    Sprite Size Mode=::NewEnumerator2
    UsePositionOffset=true
    Write Material Random=true
   )
   &ShapeLocation(MaskSpawnGroup=false, Override Seed=false, Sphere Radius=3, UseOwnerScale=true)
   &AddVelocity(
    Cone Axis=<0 1 0>
    Module.UseConeFalloff=false
    Module.UseVelDistribution=false
    Override Seed=false
    Velocity Mode=::NewEnumerator2
    Velocity Speed=&RandomRangeFloat(Minimum=-50, Maximum=150)
   )
  ]
  particleUpdateScript [
   &ParticleState()
   &GravityForce(
    Coordinate Space=::Local
    Gravity=&RandomRangeVector(
     Minimum=&MakeVector(
      X=0
      Y=&FloatFromCurve(
       FloatCurve=DataInterface() {propertyValues {Curve {
          defaultValue=1
          keys [{interpMode=RCIM_Cubic, tangentMode=RCTM_User, time=0, value=0}, {interpMode=RCIM_Cubic, tangentMode=RCTM_User, time=1, value=1}]
         }}}
       Scale Curve=-500
      )
      Z=-250
     )
     Maximum=&MakeVector(
      X=0
      Y=&FloatFromCurve(
       FloatCurve=DataInterface() {propertyValues {Curve {
          defaultValue=1
          keys [{interpMode=RCIM_Cubic, tangentMode=RCTM_User, time=0, value=0}, {interpMode=RCIM_Cubic, tangentMode=RCTM_User, time=1, value=1}]
         }}}
       Scale Curve=-750
      )
      Z=-500
     )
    )
   )
   &CurlNoiseForce(Bias Noise=false, Noise Frequency=250, Noise Strength=750, Pan Noise=false)
   &AerodynamicDrag(
    Rotation Strength:NiagaraFloat=&RandomRangeFloat(Minimum=5, Maximum=8)
    Use Linear Drag=true
    Use Rotational Drag=true
   )
   &InheritSourceMovement(
    Applied Position Scale=&RandomRangeVector(Minimum=<0.25 0.25 0.25>, Maximum=<0.75 0.75 0.75>)
    ApplyForce=false
    ApplyPosition=true
    ApplyPositionFalloff=false
   )
   &SolveForcesAndVelocity(Clamp Velocity=false, Limit Acceleration=false)
   &ScaleSpriteSize(Uniform Curve Sprite Scale=DataInterface() {propertyValues {Curve {defaultValue=0, keys [{tangentMode=RCTM_User, time=0, value=0}, <at 0.1 1>, <at 0.5 1>, {tangentMode=RCTM_User, time=1, value=0}]}}})
   &ScaleColor(
    Scale Alpha=&FloatFromCurve(
     FloatCurve=DataInterface() {propertyValues {Curve {keys [<at 0 0>,<at 0.1 1>]}}}
     Scale Curve=&Lerp_Float(A=1, B=0, Alpha=&MakeCustomFloatFromBool(Bool=$User.IsSkydiving, True Float=1, False Float=0))
    )
    Scale RGB=&MakeVectorFromLinearColorRGB(Color=&FNTimeOfDayColor(
      Night Color=<rgba 0.01 0.0167 0.05 1>
      Morning Color=<rgba 0.05 0.0406 0.0375 1>
      Day Color=<rgba 0.02 0.025 0.05 1>
      Evening Color=<rgba 0.03 0.018 0.018 1>
     ))
    ScaleA=true
    ScaleRGB=true
   )
   &AlignSpriteToMeshOrientation(SetAlignment=true, SetFacing=true)
   &SpriteRotationRate(Rotation Rate=&RandomRangeFloat(Minimum=-15, Maximum=15))
   &CameraOffset(Camera Offset Amount=&FloatFromCurve(
     FloatCurve=DataInterface() {propertyValues {Curve {
        defaultValue=1
        keys [{interpMode=RCIM_Cubic, tangentMode=RCTM_User, time=0, value=0}, {interpMode=RCIM_Cubic, tangentMode=RCTM_User, time=1, value=1}]
       }}}
     Scale Curve=10
    ))
  ]
  rendererClasses [&NiagaraSpriteRendererProperties]
  renderers [{rendererClass=&NiagaraSpriteRendererProperties}]
 }
}
summary {
 systemName=NS_Backpack_ThornyPoint_Idle
 emitters [
  {emitterName=NE_Petals_Idle, rendererClasses [&NiagaraSpriteRendererProperties]}
  {emitterName=NE_Petals_Run, rendererClasses [&NiagaraSpriteRendererProperties]}
 ]
 userVariables [
  User.bIsFrontEnd {type=NiagaraBool}
  User.bIsFrontEndPreview {type=NiagaraBool}
  User.IsSkydiving {type=NiagaraBool}
 ]
}
