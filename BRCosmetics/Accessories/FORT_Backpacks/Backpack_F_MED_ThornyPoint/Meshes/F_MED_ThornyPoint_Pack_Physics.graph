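# Physics asset for the ThornyPoint backpack cosmetic, defining capsule bodies for the pack, a strap and a hanging piece.
# A four-segment strap chain runs from a small base body, and a two-segment hanging chain joins it at the strap master, which connects to the pack.
# Master-to-pack and base-to-master joints rotate freely, while chain joints set narrow swing and twist limits.
# The first hanging segment has two constraints to the master with differing limits.
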PhysicsAsset /BRCosmetics/Accessories/FORT_Backpacks/Backpack_F_MED_ThornyPoint/Meshes/F_MED_ThornyPoint_Pack_Physics
bodies {
 dyn_hanging_piece_01 [{shapeType=Capsule, length=8.5, radius=1.5, shapeName=dyn_hanging_piece_01_capsule, center=<7 0 -1>, rotation=<rot 90 0 0>}]
 dyn_hanging_piece_02 [{shapeType=Capsule, length=8.5617, radius=1.5236, shapeName=dyn_hanging_piece_02_capsule, center=<7 0 -1>, rotation=<rot 90 0 0>}]
 dyn_strap_01 [{shapeType=Capsule, length=13, radius=5, shapeName=dyn_strap_01_capsule, center=<3 -3 0>, rotation=<rot 0 -90 0>}]
 dyn_strap_02 [{shapeType=Capsule, length=13, radius=5, shapeName=dyn_strap_02_capsule, center=<3 -3 0>}]
 dyn_strap_03 [{shapeType=Capsule, length=11, radius=5, shapeName=dyn_strap_03_capsule, center=<3 -3.5 0>}]
 dyn_strap_04 [{shapeType=Capsule, length=10, radius=5, shapeName=dyn_strap_04_capsule, center=<3 -3.5 0>}]
 dyn_strap_base [{shapeType=Capsule, length=0.505, radius=0.505, shapeName=dyn_strap_base_capsule, rotation=<rot 0 -90 0>}]
 dyn_strap_master [
  {shapeType=Capsule, length=5, radius=5, shapeName=dyn_strap_master_capsule, center=<0 -6.5 -5>, rotation=<rot 90 0 0>}
  {shapeType=Capsule, length=15, radius=15, shapeName=dyn_strap_master_capsule, center=<-17.9986 -17.9953 0>}
 ]
 pack [
  {shapeType=Capsule, length=16.1421, radius=12.6352, shapeName=pack_capsule, center=<0 11 -10>, rotation=<rot 0 0 3.5>}
  {shapeType=Capsule, length=15, radius=15, shapeName=pack_capsule, center=<-49.9678 0 4.8558>, rotation=<rot 90 90 0>}
 ]
}
constraints [
 {
  bone1Name=dyn_strap_master
  bone2Name=pack
  swing1LimitDegrees=45
  swing1Motion=Free
  swing2LimitDegrees=45
  swing2Motion=Free
  twistLimitDegrees=45
  twistMotion=Free
 }
 {bone1Name=dyn_hanging_piece_02, bone2Name=dyn_hanging_piece_01, swing1LimitDegrees=50, swing2LimitDegrees=20, twistLimitDegrees=0}
 {bone1Name=dyn_hanging_piece_01, bone2Name=dyn_strap_master, swing1LimitDegrees=20, swing2LimitDegrees=70, twistLimitDegrees=2}
 {bone1Name=dyn_strap_04, bone2Name=dyn_strap_03, swing1LimitDegrees=40, swing2LimitDegrees=10, twistLimitDegrees=0}
 {bone1Name=dyn_strap_03, bone2Name=dyn_strap_02, swing1LimitDegrees=40, swing2LimitDegrees=10, twistLimitDegrees=0}
 {bone1Name=dyn_strap_02, bone2Name=dyn_strap_01, swing1LimitDegrees=40, swing2LimitDegrees=10, twistLimitDegrees=0}
 {bone1Name=dyn_strap_01, bone2Name=dyn_strap_base, swing1LimitDegrees=40, swing2LimitDegrees=10, twistLimitDegrees=0}
 {
  bone1Name=dyn_strap_base
  bone2Name=dyn_strap_master
  swing1LimitDegrees=45
  swing1Motion=Free
  swing2LimitDegrees=45
  swing2Motion=Free
  twistLimitDegrees=45
  twistMotion=Free
 }
 {bone1Name=dyn_hanging_piece_01, bone2Name=dyn_strap_master, swing1LimitDegrees=2, swing2LimitDegrees=20, twistLimitDegrees=0}
]
