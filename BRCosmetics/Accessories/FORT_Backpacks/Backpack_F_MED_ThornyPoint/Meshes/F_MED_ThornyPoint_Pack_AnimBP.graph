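# Animation Blueprint for the ThornyPoint backpack cosmetic, adding secondary motion to its straps and roses.
# Strap bones are rotated during skydive floating, then two RigidBody nodes switch between gameplay and front-end physics assets via bIsInFrontEnd.
# Spring-driven AnimDynamics sways four rose bones, followed by constant additive strap rotations.
# The EventGraph sets BackpackRotate from spine yaw and WorldGravity from skydive state, but the AnimGraph reads neither.
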AnimBlueprint /BRCosmetics/Accessories/FORT_Backpacks/Backpack_F_MED_ThornyPoint/Meshes/F_MED_ThornyPoint_Pack_AnimBP
target_skeleton=&/F_MED_ThornyPoint_Pack_Skeleton
graphs [
 AnimGraph {nodes [
   n1 = Get bIsSkydiveFloating()
   n2 = "Transform (Modify) Bone - Bone: dyn_strap_01"!(ComponentPose=Local To Component(LocalPose=Input Pose()), bAlphaBoolEnabled:bool=@n1, Translation="0,0,0", Rotation="0,-50,0", Scale="1,1,1") -> (Pose) {
    alphaInputType=Bool
    rotationMode=BMM_Replace
    rotationSpace=BCS_BoneSpace
    translationMode=BMM_Additive
    translationSpace=BCS_BoneSpace
    alphaBoolBlend {blendInTime=5, blendOutTime=5}
    boneToModify {boneName=dyn_strap_01}
    rotation {yaw=-50}
   }
   n3 = "Transform (Modify) Bone - Bone: dyn_strap_02"!(ComponentPose=@n2, bAlphaBoolEnabled:bool=@n1, Translation="0,0,0", Rotation="0,-20,0", Scale="1,1,1") -> (Pose) {
    alphaInputType=Bool
    rotationMode=BMM_Replace
    rotationSpace=BCS_BoneSpace
    translationMode=BMM_Additive
    translationSpace=BCS_BoneSpace
    alphaBoolBlend {blendInTime=5, blendOutTime=5}
    boneToModify {boneName=dyn_strap_02}
    rotation {yaw=-20}
   }
   n4 = RigidBody!(ComponentPose=@n3, bAlphaBoolEnabled:bool=NOT Boolean(A=Get bIsInFrontEnd()), OverrideWorldGravity:Vector="0,0,-1000", ExternalForce) -> (Pose) {
    alphaInputType=Bool
    bOverrideWorldGravity=true
    overridePhysicsAsset=&/F_MED_ThornyPoint_Pack_Physics
    simulationSpace=BaseBoneSpace
    baseBoneRef {boneName=pack}
    overrideWorldGravity {z=-1000}
    simSpaceSettings {
     maxAngularAcceleration=40
     maxAngularVelocity=8
     maxLinearAcceleration=4000
     maxLinearVelocity=800
     velocityScaleZ=0.6
     worldAlpha=0.7
     externalLinearDragV=<0.2 0.2 0.2>
    }
   }
   n5 = RigidBody!(ComponentPose=@n4, bAlphaBoolEnabled=Get bIsInFrontEnd(), ExternalForce) -> (Pose) {
    alphaInputType=Bool
    bOverrideWorldGravity=true
    overridePhysicsAsset=&/F_MED_ThornyPoint_Pack_Physics_FE
    simulationSpace=BaseBoneSpace
    baseBoneRef {boneName=pack}
    overrideWorldGravity {z=-980}
    simSpaceSettings {maxAngularAcceleration=40, maxAngularVelocity=8, worldAlpha=1, externalAngularVelocity=<1 1 1>, externalLinearDragV=<1 1 1>}
   }
   n6 = Anim Dynamics - Bone: dyn_left_rose_01(ComponentPose=@n5, Alpha=1, ExternalForce="0,0,0") {
    angularDampingOverride=1
    angularSpringConstant=10000
    bAngularSpring=true
    bOverrideAngularDamping=true
    bUseGravityOverride=true
    gravityScale=0.35
    simulationSpace=RootRelative
    boundBone {boneName=dyn_left_rose_01}
    componentAppliedLinearAccClamp=<50000 50000 50000>
    componentLinearAccScale=<0.5 0.5 0.5>
    componentLinearVelScale=<0.5 0.5 5>
    gravityOverride {z=-980}
    physicsBodyDefinitions [{
      boundBone {boneName=dyn_left_rose_01}
      boxExtents=<20 30 30>
      constraintSetup {angularLimitsMax=<0 10 10>, angularLimitsMin=<0 -10 -10>, angularTarget=<1 0 0>}
      localJointOffset=<4 0 0>
     }]
   }
   n7 = Anim Dynamics - Bone: dyn_b_right_rose_01(ComponentPose=@n6, Alpha=1, ExternalForce="0,0,0") {
    angularDampingOverride=1
    angularSpringConstant=10000
    bAngularSpring=true
    bOverrideAngularDamping=true
    bUseGravityOverride=true
    gravityScale=0.35
    simulationSpace=RootRelative
    boundBone {boneName=dyn_b_right_rose_01}
    componentAppliedLinearAccClamp=<50000 50000 50000>
    componentLinearAccScale=<0.5 0.5 0.5>
    componentLinearVelScale=<0.5 0.5 5>
    gravityOverride {z=-980}
    physicsBodyDefinitions [{
      boundBone {boneName=dyn_b_right_rose_01}
      boxExtents=<20 30 30>
      constraintSetup {angularLimitsMax=<0 10 10>, angularLimitsMin=<0 -10 -10>, angularTarget=<1 0 0>}
      localJointOffset=<4 0 0>
     }]
   }
   n8 = Anim Dynamics - Bone: dyn_a_right_rose_01(ComponentPose=@n7, Alpha=1, ExternalForce="0,0,0") {
    angularDampingOverride=1
    angularSpringConstant=10000
    bAngularSpring=true
    bOverrideAngularDamping=true
    bUseGravityOverride=true
    gravityScale=0.35
    simulationSpace=RootRelative
    boundBone {boneName=dyn_a_right_rose_01}
    componentAppliedLinearAccClamp=<5000 5000 5000>
    componentLinearAccScale=<0.5 0.5 0.5>
    componentLinearVelScale=<0.5 0.5 5>
    gravityOverride {z=-980}
    physicsBodyDefinitions [{
      boundBone {boneName=dyn_a_right_rose_01}
      boxExtents=<20 30 30>
      constraintSetup {angularLimitsMax=<0 10 10>, angularLimitsMin=<0 -10 -10>, angularTarget=<1 0 0>}
      localJointOffset=<2 0 0>
     }]
   }
   n9 = Anim Dynamics - Bone: dyn_mid_rose_01(ComponentPose=@n8, Alpha=1, ExternalForce="0,0,0") {
    angularDampingOverride=1
    angularSpringConstant=10000
    bAngularSpring=true
    bOverrideAngularDamping=true
    bUseGravityOverride=true
    gravityScale=0.35
    simulationSpace=RootRelative
    boundBone {boneName=dyn_mid_rose_01}
    componentAppliedLinearAccClamp=<50000 50000 50000>
    componentLinearAccScale=<0.5 0.5 0.5>
    componentLinearVelScale=<0.5 0.5 5>
    gravityOverride {z=-980}
    physicsBodyDefinitions [{
      boundBone {boneName=dyn_mid_rose_01}
      boxExtents=<20 30 30>
      constraintSetup {angularLimitsMax=<0 10 10>, angularLimitsMin=<0 -10 -10>, angularTarget=<1 0 0>}
      localJointOffset=<2 0 0>
     }]
   }
   n10 = "Transform (Modify) Bone - Bone: dyn_strap_01"!(ComponentPose=@n9, Alpha:float=1, Translation="0,0,0", Rotation="0,-60,0", Scale="1,1,1") -> (Pose) {
    rotationMode=BMM_Additive
    rotationSpace=BCS_BoneSpace
    scaleMode=BMM_Additive
    scaleSpace=BCS_BoneSpace
    boneToModify {boneName=dyn_strap_01}
    rotation {yaw=-60}
   }
   n11 = "Transform (Modify) Bone - Bone: dyn_strap_02"!(ComponentPose=@n10, Alpha:float=1, Translation="0,0,0", Rotation="0,-60,0", Scale="1,1,1") -> (Pose) {
    rotationMode=BMM_Additive
    rotationSpace=BCS_BoneSpace
    scaleMode=BMM_Additive
    scaleSpace=BCS_BoneSpace
    boneToModify {boneName=dyn_strap_02}
    rotation {yaw=-60}
   }
   n12 = "Transform (Modify) Bone - Bone: dyn_strap_03"(ComponentPose=@n11, Alpha=1, Translation="0,0,0", Rotation="0,-40,0", Scale="1,1,1") {
    rotationMode=BMM_Additive
    rotationSpace=BCS_BoneSpace
    scaleMode=BMM_Additive
    scaleSpace=BCS_BoneSpace
    boneToModify {boneName=dyn_strap_03}
    rotation {yaw=-40}
   }
   Output Pose(Result=Component To Local(ComponentPose=@n12)) {layerGroup=None}
  ]}
 EventGraph {nodes [
   Event BlueprintUpdateAnimation()
   > n1 = Sequence()
   n2 = BreakTransform(InTransform=GetDeltaTransformFromRefPose(self=GetOwnerSkeletalMeshComponent(), BoneName=spine_05, BaseName=spine_01))
   > Set BackpackRotate(BackpackRotate:double=MapRangeClamped(Value=BreakRotator(InRot=@n2.Rotation).Yaw, InRangeA=-37, InRangeB=-65)) -> (Output_Get:double)
   n3 = Branch(execute=@n1.then_1, Condition=Get bIsSkydiveFloating())
   > Set WorldGravity(WorldGravity:Vector="0,-100,600") -> (Output_Get:Vector)
   n4 = Branch(execute=@n3.else, Condition=Get bIsSkydiveDiveMode())
   > Set WorldGravity(WorldGravity="-500,0,700")
   Set WorldGravity(execute=@n4.else, WorldGravity="0,0,-980")
   Reroute Node(InputPin:Rotator=@n2.Rotation) -> (OutputPin:Rotator)
  ]}
]
variables [
 BackpackRotate {type=number}
 NewVar {type=boolean}
 WorldGravity {type=Vector}
]
